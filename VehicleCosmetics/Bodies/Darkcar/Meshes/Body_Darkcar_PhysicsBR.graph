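# Physics asset for the Darkcar vehicle cosmetic body, in what its name marks as a BR variant, providing a simplified collision shape.
# The body_con bone carries one large capsule, tipped nearly horizontal, as the asset's only shape.
# The root body has no shapes and there are no constraints, so the asset defines no jointed or ragdoll setup.
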PhysicsAsset /VehicleCosmetics/Bodies/Darkcar/Meshes/Body_Darkcar_PhysicsBR
bodies {body_con [{shapeType=Capsule, length=274.7665, radius=153.1448, center=<-2.7016 0.03 -5.3783>, rotation=<rot 0 -89.9903 -85.229>}], root []}
constraints []
